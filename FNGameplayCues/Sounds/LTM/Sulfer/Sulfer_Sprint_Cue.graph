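# Looping wind sound cue for the Sulfer LTM sprint, built from skydive wind loops and using glider sound class and attenuation.
# A mixer blends a constant horizontal wind loop, a third-person skydive loop pitched up to roughly double, and four directional layers.
# LeftParam, RightParam, FrontParam and BackParam crossfade bank-left, bank-right, vertical and horizontal loops, each fading in at low values and out at high values.
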SoundCue /FNGameplayCues/Sounds/LTM/Sulfer/Sulfer_Sprint_Cue
root=Mixer_0
nodes [
 {
  id=Mixer_0
  children [WavePlayer_14, Modulator_0, ParamCrossFade_2, ParamCrossFade_3, ParamCrossFade_4, ParamCrossFade_5]
  properties {inputVolume [0.8,1.3093,0.5,0.5,1,0.7]}
 }
 {id=Modulator_0, children [WavePlayer_0], properties {pitchMax=1.8022, pitchMin=2.1172}}
 {
  id=ParamCrossFade_2
  children [WavePlayer_8]
  properties {
   paramName=LeftParam
   crossFadeInput [{fadeInDistanceEnd=0.25, fadeInDistanceStart=0, fadeOutDistanceEnd=1, fadeOutDistanceStart=0.75, volume=1}]
  }
 }
 {
  id=ParamCrossFade_3
  children [WavePlayer_12]
  properties {
   paramName=RightParam
   crossFadeInput [{fadeInDistanceEnd=0.25, fadeInDistanceStart=0, fadeOutDistanceEnd=1, fadeOutDistanceStart=0.75, volume=1}]
  }
 }
 {
  id=ParamCrossFade_4
  children [WavePlayer_7]
  properties {
   paramName=FrontParam
   crossFadeInput [{fadeInDistanceEnd=0.25, fadeInDistanceStart=0, fadeOutDistanceEnd=1, fadeOutDistanceStart=0.75, volume=1}]
  }
 }
 {
  id=ParamCrossFade_5
  children [WavePlayer_5]
  properties {
   paramName=BackParam
   crossFadeInput [{fadeInDistanceEnd=0.25, fadeInDistanceStart=0, fadeOutDistanceEnd=1, fadeOutDistanceStart=0.75, volume=1}]
  }
 }
 {id=WavePlayer_0, properties {bLooping=true, soundWaveAssetPtr=/Game/Athena/Sounds/Wind_Falling/sfx_parachute_skydive_loop_3p}}
 {id=WavePlayer_12, properties {bLooping=true, soundWaveAssetPtr=/Game/Athena/Sounds/Wind_Falling/AMB_Wind_Skydive_Bank_Right_Loop_01}}
 {id=WavePlayer_14, properties {bLooping=true, soundWaveAssetPtr=/Game/Athena/Sounds/Wind_Falling/AMB_Wind_Skydive_Horizontal_Loop_01}}
 {id=WavePlayer_5, properties {bLooping=true, soundWaveAssetPtr=&AMB_Wind_Skydive_Horizontal_Loop_01}}
 {id=WavePlayer_7, properties {bLooping=true, soundWaveAssetPtr=/Game/Athena/Sounds/Wind_Falling/AMB_Wind_Skydive_Vertical_Loop_01}}
 {id=WavePlayer_8, properties {bLooping=true, soundWaveAssetPtr=/Game/Athena/Sounds/Wind_Falling/AMB_Wind_Skydive_Bank_Left_Loop_01}}
]
properties {
 attenuationSettings=/Game/Sounds/Attenuation/Fort_Player_Glider_OpenClose_Athena_Att
 maxDistance=10000
 soundClassObject=&Fort_Gliders_InAir
 volumeMultiplier=0.35
}
